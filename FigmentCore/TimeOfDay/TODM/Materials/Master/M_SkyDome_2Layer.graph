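# Unlit sky dome master material with two animated cloud layers over a gradient sky, masked stars, and a glowing sun/moon.
# Swirl-distorted, panned cloud textures, the lower with bump-offset parallax, height-blend into coverage and bent normals.
# Material-collection sun/moon direction drives cloud lighting, offset-sample shadows and glow, while sky-atmosphere light direction places the disc.
# An optional static switch radially blends volcano cloud colours around a centre point.
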Material /FigmentCore/TimeOfDay/TODM/Materials/Master/M_SkyDome_2Layer
graph {
 properties {bIsSky=true, bTangentSpaceNormal=false, pixelDepthOffsetMode=PDOM_Legacy, shadingModel=MSM_Unlit, usedShadingModels=Unlit}
 nodes [
  n1 = ScalarParameter() {defaultValue=-0.1619, group=Lighting, parameterName=LightPowerCloudBleed}
  n2 = Power(Base=OneMinus(), Exp=@n1)
  Clamp(None=@n2)
  n3 = TextureSample(UVs=Multiply(A=TextureCoordinate(), B=ScalarParameter() {defaultValue=50, parameterName=TilingSmallNoise})) {samplerType=SAMPLERTYPE_Masks, texture=&/CloudShape4_N}
  ComponentMask(None=@n3.RGB) {g=true, r=true}
  Constant2Vector() {r=-1}
  n4 = CollectionParameter() {
   collection=&FortniteMaterialParameters
   parameterName=SunAndMoonModelDirectionalVector
  }
  n5 = ComponentMask(None=@n4) {b=true, g=true, r=true}
  n6 = CollectionParameter() {
   collection=&FortniteMaterialParameters
   parameterName=SunlightYTransformVector
  }
  n7 = CollectionParameter() {
   collection=&FortniteMaterialParameters
   parameterName=SunAndMoonModelWorldPosition
  }
  MaterialFunctionCall(
   Plane Normal Axis=@n5
   Plane X Axis=ComponentMask(None=@n6) {b=true, g=true, r=true}
   Plane Center=ComponentMask(None=@n7) {b=true, g=true, r=true}
   UV Scale=ScalarParameter() {defaultValue=5374515, group=SunSettings, parameterName=SunMoon-Scale}
  ) {materialFunction=&VirtualPlaneCoordinates}
  n8 = NamedRerouteUsage()
  n9 = ComponentMask(None=SkyAtmosphereLightDirection()) {b=true, g=true, r=true}
  NamedRerouteDeclaration(None=Step(X=DotProduct(A=@n8, B=@n9))) {name=HemisphereMask, nodeColor=<rgba 1 0 0.3047 1>}
  n10 = WorldPosition()
  n11 = CameraPositionWS()
  NamedRerouteDeclaration(None=Normalize(VectorInput=Subtract(A=@n10.XYZ, B=@n11))) {name=CameraVectorForVertexShader, nodeColor=<rgba 0 1 0.0859 1>}
  OneMinus(None=@n2)
  n12 = Multiply(A=@n9) {constB=-1}
  n13 = CrossProduct(A=@n12, B=Constant3Vector() {constant=<rgba 0 0 -1 0>})
  n14 = DotProduct(A=Normalize(VectorInput=CrossProduct(A=@n13, B=@n12)), B=NamedRerouteUsage())
  n15 = DotProduct(A=Normalize(VectorInput=@n13), B=NamedRerouteUsage())
  n16 = Divide(
   A=NamedRerouteDeclaration(None=AppendVector(A=@n14, B=@n15)) {name=SunMoonUVs, nodeColor=<rgba 1 0 0.9141 1>}
   B=ScalarParameter() {defaultValue=0.07, group=Sun-Moon, parameterName=MoonSize}
  )
  n17 = TextureSample(UVs=VertexInterpolator(VS=Add(A=@n16) {constB=0.5}).PS) {samplerType=SAMPLERTYPE_LinearColor, texture=/Game/Packages/Fortress_Sky/SkyDome/Master/T_SunMoon}
  n18 = Multiply(A=ScalarParameter() {defaultValue=-100, group=SunSettings, parameterName=Moon-PreBlendOffset}, B=@n17.R)
  n19 = Multiply(
   A=ScalarParameter() {defaultValue=8, group=SunSettings, parameterName=Sun-SharpnessMod}
   B=Power(Base=@n17.G, Exp=ScalarParameter() {defaultValue=40.8355, group=SunSettings, parameterName=Sun-SharpnessPower})
  )
  n20 = Add(
   A=Multiply(A=ScalarParameter() {group=SunSettings, parameterName=Moon-Amount}, B=Saturate(None=@n18))
   B=Multiply(A=ScalarParameter() {defaultValue=40, group=SunSettings, parameterName=Sun-Amount}, B=Saturate(None=@n19))
  )
  n21 = VectorParameter() {group=SunSettings, parameterName=SunMoon-Color, defaultValue=<rgba 3 1.8818 0.45 1>}
  n22 = Multiply(A=DotProduct(A=@n5, B=CameraVectorWS())) {constB=-1}
  n23 = Multiply(
   A=Power(Base=@n22, Exp=ScalarParameter() {defaultValue=80, group=SunSettings, parameterName=SunMoon-GlowPower})
   B=ScalarParameter() {defaultValue=0.25, group=SunSettings, parameterName=SunMoon-GlowMod}
  )
  n24 = VectorParameter() {group=SunSettings, parameterName=SunMoon-GlowColor, defaultValue=<rgba 1 0.9505 0.9 1>}
  n25 = Multiply(A=@n22, B=ScalarParameter() {defaultValue=50, group=SunSettings, parameterName=SunMoon-UVErrorMaskMod})
  n26 = ScalarParameter() {defaultValue=-20, group=SunSettings, parameterName=SunMoon-UVErrorMaskAdd}
  n27 = Multiply(A=Add(A=Multiply(A=@n20, B=@n21.RGB), B=Multiply(A=@n23, B=@n24.RGB)), B=Saturate(None=Add(A=@n25, B=@n26)))
  Reroute(None=Multiply(A=Constant(), B=@n27))
  n28 = ComponentMask(None=TextureCoordinate() {coordinateIndex=2}) {g=true}
  n29 = LinearInterpolate(
   A=VectorParameter() {group=Colors, parameterName=Sky-ZenithColor, defaultValue=<rgba 0.02 0.0859 0.2 1>}.RGB
   B=VectorParameter() {group=Colors, parameterName=Sky-HorizonColor, defaultValue=<rgba 0.25 1.0732 2.5 1>}.RGB
   Alpha=Power(Base=@n28, Exp=ScalarParameter() {defaultValue=9.3522, group=Colors, parameterName=Sky-HorizonPower})
  )
  n30 = ScalarParameter() {defaultValue=1, group=Colors, parameterName=Sky-LowHorizon-AddBright}
  n31 = ScalarParameter() {defaultValue=10, group=Colors, parameterName=Sky-LowHorizon-Power}
  n32 = Add(A=@n29, B=Multiply(A=Multiply(A=@n29, B=@n30), B=Power(Base=@n28, Exp=@n31)))
  Reroute(None=@n32)
  ScalarParameter() {parameterName=test}
  n33 = VectorParameter() {parameterName=ColorVolcanoCloud, defaultValue=<rgba 0.0069 0.0083 0.01 1>}
  n34 = VectorParameter() {parameterName=ColorVolcanoCloud2, defaultValue=<rgba 0.05 0.0073 0 1>}
  n35 = VertexNormalWS()
  n36 = ScalarParameter() {defaultValue=0.2, group=Normals, parameterName=BendNormals}
  n37 = AppendVector(A=AppendVector(A=@n36, B=@n36), B=Constant())
  n38 = ScalarParameter() {defaultValue=1.4445, group=Parent-MI-Settings, parameterName=SwirlTile}
  n39 = Multiply(A=TextureCoordinate(), B=ScalarParameter() {defaultValue=5.1497, group=Parent-MI-Settings, parameterName=Cloud-TileMaster}) {constB=0.5}
  n40 = ScalarParameter() {defaultValue=0.781, group=Parent-MI-Settings, parameterName=Cloud-SpeedMaster}
  n41 = Multiply(A=ScalarParameter() {defaultValue=0.01, group=Parent-MI-Settings, parameterName=SwirlPanSpeed}, B=@n40)
  n42 = Add(
   A=MaterialFunctionCall!(
    SineAnimOffset
    Flow Direction=Constant2Vector() {g=0.1, r=-1}
    Time offset
    EdgeMaskSize
    UVS * size=Multiply(A=@n39, B=@n38) {constB=0.5}
    SineAnimSpeed=ScalarParameter() {defaultValue=1.5, group=Parent-MI-Settings, parameterName=SwirlCurlSpeed}
    Grid Noise Tile Rep
    Rotation amount modulator=ScalarParameter() {defaultValue=0.25, group=Parent-MI-Settings, parameterName=SwirlAmount}
    un frac Time=MaterialFunctionCall(Speed=@n41) {materialFunction=&TimeWithSpeedVariable}.Time
   ) {desc=15 fewer instructions on low end, materialFunction=/Game/Packages/Fortress_Sky/Fortnite_SkyTwistFunction}.Rotated UVs
   B=@n39
  )
  n43 = Multiply(A=@n3.B, B=ScalarParameter() {defaultValue=0.01, parameterName=Distort})
  n44 = Panner(
   Coordinate=Add(A=@n42, B=@n43)
   Speed=Multiply(A=ScalarParameter() {defaultValue=0.01, group=PanningParalaxing, parameterName=PanSpeedTopLayer}, B=@n40)
  )
  n45 = Add(A=Multiply(A=@n42, B=ScalarParameter() {defaultValue=0.5, group=PanningParalaxing, parameterName=BottomLayerTilingOffset}), B=@n43)
  n46 = Multiply(A=@n40, B=ScalarParameter() {defaultValue=0.01, group=PanningParalaxing, parameterName=PanSpeedBottomLayer})
  n47 = TextureSample(UVs=BumpOffset(
    Coordinate=Panner(Coordinate=@n45, Speed=@n46)
    Height=Constant() {r=1}
    HeightRatioInput=ScalarParameter() {defaultValue=-1, group=PanningParalaxing, parameterName=BumpOffset}
   )) {samplerType=SAMPLERTYPE_Masks, texture=&CloudShape4_N}
  n48 = MaterialFunctionCall(
   Transition Phase=ScalarParameter() {defaultValue=0.418, group=CloudShape, parameterName=CloudShapeTransition2}
   Height Texture=@n47.B
   Contrast=ScalarParameter() {defaultValue=2, group=CloudShape, parameterName=CloudShape2}
  ) {materialFunction=&HeightLerp}
  n49 = LinearInterpolate(
   A=TextureSample(UVs=@n44) {samplerType=SAMPLERTYPE_Masks, texture=&CloudShape4_N}.RGB
   B=@n47.RGB
   Alpha=@n48.Alpha
  )
  n50 = ConstantBiasScale(None=ComponentMask(None=@n49) {g=true, r=true}) {bias=-0.5, scale=2}
  n51 = Multiply(
   A=Add(A=Multiply(A=@n35, B=@n37), B=DeriveNormalZ(InXY=@n50))
   B=VectorParameter() {group=Normals, parameterName=Cloud-NormalsMod, defaultValue=<rgba 1 1 -1 1>}.RGB
  )
  n52 = Normalize(VectorInput=CrossProduct(A=@n51, B=Constant3Vector() {constant=<rgba 0 0 1 0>}))
  n53 = DotProduct(A=@n49, B=Constant3Vector() {constant=<rgba 0 0 1 0>})
  n54 = Multiply(
   A=Power(Base=OneMinus(None=@n53), Exp=ScalarParameter() {defaultValue=2, group=Normals, parameterName=PuffyExponentNormal})
   B=ScalarParameter() {defaultValue=1.5, group=Normals, parameterName=RotationAngle}
  )
  n55 = Constant3Vector()
  n56 = Normalize(VectorInput=Add(A=RotateAboutAxis(NormalizedRotationAxis=@n52, RotationAngle=@n54, PivotPoint=@n55, Position=@n51), B=@n51))
  n57 = Constant3Vector() {constant=<rgba 0 0 -1 0>}
  n58 = Saturate(None=Power(Base=DotProduct(A=@n56, B=@n57), Exp=ScalarParameter() {defaultValue=4, parameterName=FalloffVolcanoLight}))
  n59 = ScalarParameter() {defaultValue=1, group=SunSettings, parameterName=SunMoon-AtmospherePower}
  n60 = Saturate(None=Multiply(A=Power(Base=@n22, Exp=@n59), B=ScalarParameter() {defaultValue=0.75, group=SunSettings, parameterName=SunMoon-AtmosphereMod}))
  n61 = LinearInterpolate(
   A=@n32
   B=VectorParameter() {group=SunSettings, parameterName=SunMoon-AtmosphereColor, defaultValue=<rgba 0.4792 0.3872 0.2604 1>}.RGB
   Alpha=@n60
  )
  n62 = ScalarParameter() {defaultValue=16, group=CloudShape, parameterName=SunMoonCloudFadeArea}
  n63 = Multiply(A=Clamp(None=Power(Base=@n22, Exp=@n62)), B=ScalarParameter() {defaultValue=0.2762, group=CloudShape, parameterName=SunMoonCloudFade})
  n64 = MaterialFunctionCall(
   Transition Phase=Multiply(A=ScalarParameter() {defaultValue=0.47, group=CloudShape, parameterName=CloudShapeTransition}, B=OneMinus(None=@n63))
   Height Texture=ComponentMask(None=@n49) {b=true}
   Contrast=ScalarParameter() {defaultValue=2, group=CloudShape, parameterName=CloudShape}
  ) {materialFunction=&HeightLerp}
  n65 = TextureCoordinate() {coordinateIndex=1}
  n66 = TextureSampleParameter2D(UVs=Multiply(A=@n65, B=ScalarParameter() {defaultValue=7, group=Stars, parameterName=StarTile})) {parameterName=CloudandStars, samplerType=SAMPLERTYPE_LinearColor, texture=/Game/Packages/Fortress_Sky/SkyDome/Master/T_SkyDome_StarCloudMask}
  n67 = ScalarParameter() {defaultValue=15, group=Stars, parameterName=StarMaskTile}
  n68 = ScalarParameter() {defaultValue=0.05, group=Stars, parameterName=StarMaskSpeed}
  n69 = TextureSampleParameter2D(UVs=Panner(
    Coordinate=Multiply(A=@n65, B=@n67)
    Time=MaterialFunctionCall(Speed=@n68) {materialFunction=&TimeWithSpeedVariable}.Time
   ) {speedX=-1}) {parameterName=CloudandStars, samplerType=SAMPLERTYPE_LinearColor, texture=&T_SkyDome_StarCloudMask}
  n70 = Multiply(A=@n66.R, B=@n69.G)
  n71 = ScalarParameter() {defaultValue=3, group=Stars, parameterName=StarPower}
  n72 = Multiply(A=OneMinus(None=@n64.Alpha), B=Multiply(A=Power(Base=@n70, Exp=@n71), B=ScalarParameter() {group=Stars, parameterName=StarBrightness}))
  n73 = VectorParameter() {group=Colors, parameterName=Cloud-ShadowColor, defaultValue=<rgba 0.0252 0.0252 0.0252 1>}
  n74 = CollectionParameter() {
   collection=&FortniteMaterialParameters
   parameterName=SunAndMoonModelDirectionalVector
  }
  n75 = Multiply(A=ComponentMask(None=@n74) {g=true, r=true}, B=ScalarParameter() {defaultValue=0.1333, group=Lighting, parameterName=OffsetCloudShadow})
  n76 = ComponentMask(None=TextureSample(UVs=Add(A=@n44, B=@n75)) {samplerType=SAMPLERTYPE_Masks, texture=&CloudShape4_N}.B) {b=true}
  n77 = ScalarParameter() {defaultValue=0.4571, group=Lighting, parameterName=CloudVolumePower}
  n78 = CollectionParameter() {
   collection=&FortniteMaterialParameters
   parameterName=SunAndMoonModelDirectionalVector
  }
  n79 = ComponentMask(None=@n78) {b=true, g=true, r=true}
  n80 = Normalize(VectorInput=Multiply(A=@n79, B=Constant3Vector() {constant=<rgba -1 -1 0 0>}))
  n81 = LinearInterpolate(
   A=LinearInterpolate(
    A=VectorParameter() {group=Colors, parameterName=Cloud-BounceColorZenith, defaultValue=<rgba 0.1746 0.2232 0.2542 1>}.RGB
    B=@n73.RGB
    Alpha=Power(Base=@n76, Exp=@n77)
   )
   B=@n73.RGB
   Alpha=Power(Base=DotProduct(A=@n80, B=@n56), Exp=ScalarParameter() {defaultValue=4, group=Lighting, parameterName=BounceLightPower})
  )
  n82 = VectorParameter() {group=Colors, parameterName=Cloud-DirectSunLightColor, defaultValue=<rgba 1 0.7011 0.4852 1>}
  n83 = DotProduct(A=Normalize(VectorInput=@n79), B=@n56)
  n84 = MaterialFunctionCall(In=@n83, Contrast=@n1) {materialFunction=&CheapContrast}
  n85 = ScalarParameter() {defaultValue=1, group=Lighting, parameterName=LightBrightnessCloudBleed}
  n86 = ScalarParameter() {defaultValue=2, group=Lighting, parameterName=LightPower}
  n87 = Add(A=Multiply(A=Multiply(A=@n84.Result, B=@n85), B=@n60), B=Multiply(A=Power(Base=@n83, Exp=@n86), B=OneMinus(None=@n48.Alpha)))
  n88 = ScalarParameter() {defaultValue=1, group=Lighting, parameterName=BrightnessTopLayer}
  n89 = ScalarParameter() {defaultValue=1, group=Lighting, parameterName=Contrast2ndLayerCloudShadow}
  n90 = ScalarParameter() {defaultValue=-0.2298, group=Lighting, parameterName=Contrast2ndLayerCloudShadow2}
  n91 = Clamp(None=Multiply(A=LinearInterpolate(A=@n89, B=@n90, Alpha=@n76), B=ScalarParameter() {defaultValue=1.2, group=Lighting, parameterName=BrightnessBottomLayer}))
  n92 = ScalarParameter() {defaultValue=0.8, group=CloudOpacitySettings, parameterName=1-FinalCloudOpacityBoost}
  n93 = TextureCoordinate() {coordinateIndex=2}
  n94 = Clamp(None=Power(Base=ComponentMask(None=@n93) {g=true}, Exp=ScalarParameter() {defaultValue=50, group=Colors, parameterName=Sky-HorizonPowerWash}))
  n95 = LinearInterpolate(
   A=Add(A=@n27, B=Add(A=@n61, B=@n72))
   B=Multiply(A=LinearInterpolate(A=@n81, B=@n82.RGB, Alpha=@n87), B=LinearInterpolate(A=@n88, B=@n91, Alpha=@n48.Alpha))
   Alpha=Multiply(A=Multiply(A=@n64.Alpha, B=@n92), B=OneMinus(None=@n94))
  )
  n96 = WorldPosition()
  n97 = VectorParameter() {parameterName=CenterLocation, defaultValue=<rgba 0 0 0 1>}
  n98 = Divide(
   A=Distance(A=ComponentMask(None=@n96.XYZ) {g=true, r=true}, B=ComponentMask(None=@n97.RGB) {g=true, r=true})
   B=ScalarParameter() {defaultValue=3241973, parameterName=DistanceRadialBlend}
  )
  StaticSwitchParameter(True=LinearInterpolate(A=LinearInterpolate(A=@n33.RGB, B=@n34.RGB, Alpha=@n58), B=@n95, Alpha=Saturate(None=@n98)), False=@n95) {parameterName=UseVolcanoCloudBlend}
 ]
}
